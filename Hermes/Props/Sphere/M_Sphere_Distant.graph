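# M_Sphere_Distant is a minimal Default Lit material, named for a distant version of the Sphere prop.
# It holds two constant nodes, a near-black neutral grey color and a high scalar, with no textures or parameters.
# Automatic usage setting in the editor is turned off.
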Material /Hermes/Props/Sphere/M_Sphere_Distant
graph {
 properties {bAutomaticallySetUsageInEditor=false, usedShadingModels=Default Lit}
 nodes [
  Constant3Vector() {constant=<rgba 0.0104 0.0104 0.0104 0>}
  Constant() {r=0.8}
 ]
}
